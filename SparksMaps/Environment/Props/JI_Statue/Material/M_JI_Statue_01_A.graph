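# M_JI_Statue_01_A is a Default Lit material for the JI_Statue environment prop, flagged for use with Nanite.
# It holds three texture samples from the statue's own texture set: a D map, an N normal map and an S map sampled as linear color.
# The graph has no math or parameters, and no links from the samples to material outputs appear.
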Material /SparksMaps/Environment/Props/JI_Statue/Material/M_JI_Statue_01_A
graph {
 properties {bUsedWithNanite=true, pixelDepthOffsetMode=PDOM_Legacy, usedShadingModels=Default Lit}
 nodes [
  TextureSample() {texture=/SparksMaps/Environment/Props/JI_Statue/Textures/T_JI_Statue_Texture_D}
  TextureSample() {samplerType=SAMPLERTYPE_Normal, texture=/SparksMaps/Environment/Props/JI_Statue/Textures/T_JI_Statue_Texture_N}
  TextureSample() {samplerType=SAMPLERTYPE_LinearColor, texture=/SparksMaps/Environment/Props/JI_Statue/Textures/T_JI_Statue_Texture_S}
 ]
}
